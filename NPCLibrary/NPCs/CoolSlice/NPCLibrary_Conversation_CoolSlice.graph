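# Conversation database for the CoolSlice NPC, defining its greeting and the dialogue branches reachable from it.
# The entry point leads to one speech node whose default line is overridden by prioritized contextual messages for first-ever meetings, conversation-count checks and two specific player outfits.
# That speech node fans out through link nodes to remote entry tags for NPC library quests, basic services and two Season 38 story quest conversations.
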ConversationDatabase /NPCLibrary/NPCs/CoolSlice/NPCLibrary_Conversation_CoolSlice
entry_points [{tag=Conversation.Entry.CoolSlice, destinations [#1]}]
nodes [
 ConversationEntryPointNode_7 {
  id=#1
  class=ConversationEntryPointNode
  outputs [#2]
  properties {entryTag {tagName=Conversation.Entry.CoolSlice}}
 }
 Speech_C_4 {
  id=#2
  class=Speech_C
  outputs [#3, #4, #5, #6]
  properties {
   nodeName=""
   generalConfig {
    defaultMessage=What can I ding-dong-diddly-do for you?
    contextualMessages [
     {
      message="Hi-diddly-hey, neighborino!"
      contextRequirements [{requirement {class=FortControllerRequirement_IsFirstEverConversationWithNPC}}]
      priority {curve {rowName=Tandem.Speech.Priority.FirstEverInteraction}}
     }
     {
      message=What can I ding-dong-diddly-do for you?
      contextRequirements [{requirement {class=FortControllerRequirement_CheckNPCMatchConversationCount, properties {nPCUniqueIdTag {tagName=AISpawnerData.Type.Tandem.CoolSlice}}}}]
      priority {curve {rowName=None}}
     }
     {
      message="Can't you give me one ding-dang moment of peace!"
      contextRequirements [{requirement {class=FortControllerRequirement_HasCID, properties {allowedCIds [/BRCosmetics/Athena/Items/Cosmetics/Characters/Character_SilentNovel_Vane]}}}]
      priority {value=2, curve {rowName=Tandem.Speech.Priority.OutfitReaction}}
     }
     {
      message="Tell Homer enough with the ding-dong-diddly sky donuts!"
      contextRequirements [{requirement {class=FortControllerRequirement_HasCID, properties {allowedCIds [/BRCosmetics/Athena/Items/Cosmetics/Characters/Character_FruitFire_Nudge]}}}]
      priority {value=2, curve {rowName=Tandem.Speech.Priority.OutfitReaction}}
     }
    ]
   }
  }
 }
 ConversationLinkNode_3 {
  id=#6
  class=ConversationLinkNode
  properties {remoteEntryTag {tagName=Conversation.Entry.Season38.Story.P02.Q05.02}}
 }
 ConversationLinkNode_2 {
  id=#5
  class=ConversationLinkNode
  properties {remoteEntryTag {tagName=Conversation.Entry.Season38.Story.P02.Q02.02}}
 }
 ConversationLinkNode_0 {id=#4, class=ConversationLinkNode, properties {remoteEntryTag {tagName=Conversation.Entry.BasicServices}}}
 ConversationLinkNode_1 {id=#3, class=ConversationLinkNode, properties {remoteEntryTag {tagName=Conversation.Entry.NPCLibrary.Quests}}}
]
